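# Master translucent, unlit toon-outline material for the Teriyaki Fish Toon character, drawing inner lines over the mesh.
# Four-neighbor SceneDepth edge detection, optionally combined with roughness GBuffer edges, drives opacity through GBuffer and depth masks, with distance-scaled line width.
# Forward and mobile paths use depth only, behind a static switch.
# Outline color feeds emissive alongside an optional sweep, and WPO inflates along normals with optional bob and rotation, routed through MF_TechArtOverrides.
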Material /Game/Characters/Player/Male/Medium/Bodies/M_MED_Teriyaki_Fish_Toon/Materials/M_TeriyakiFishToon_Master
graph {
 properties {
  bEnableResponsiveAA=true
  bUseMaterialAttributes=true
  bUsedWithMeshDeformer=true
  bUsedWithMorphTargets=true
  bUsedWithNiagaraMeshParticles=true
  bUsedWithSkeletalMesh=true
  blendMode=BLEND_Translucent
  pixelDepthOffsetMode=PDOM_Legacy
  shadingModel=MSM_Unlit
  twoSided=true
  usedShadingModels=Unlit
 }
 nodes [
  Constant() {r=1}
  Constant() {r=0.001}
  Constant() {r=2}
  Constant() {r=0.75}
  Constant()
  Constant() {desc=for PS4 and XO, r=1}
  MaterialFunctionCall() {materialFunction=&ComponentPivotLocation}
  n1 = VectorParameter() {parameterName=OutlineColor, defaultValue=<rgba 0 0 0 1>}
  n2 = Constant()
  n3 = VectorParameter() {parameterName=MobileOutlineColor}
  n4 = Add(
   A=NamedRerouteUsage()
   B=ShadingPathSwitch(Default=StaticSwitchParameter(True=@n1.RGB, False=@n2) {parameterName=UseOutlineColor}, Forward=@n3.RGB, Mobile=@n3.RGB)
  )
  n5 = SceneTexture() {sceneTextureId=PPI_SubsurfaceColor}
  n6 = Constant()
  n7 = If(
   A=Min(A=ComponentMask(None=@n5.Color) {g=true}, B=ComponentMask(None=@n5.Color) {b=true})
   B=SquareRoot(None=Constant() {r=0.001})
   "A > B"=@n6
   "A == B"=Constant() {r=1}
   "A < B"=@n6
  ) {equalsThreshold=0.0005}
  n8 = Floor(None=SceneTexture() {sceneTextureId=PPI_Opacity}.Color)
  n9 = PixelDepth()
  n10 = SceneDepth() {desc=SceneDepth works on mobile whereas SceneTexture doesn't}
  n11 = ScalarParameter() {defaultValue=3, group=Cartoon, parameterName=CRTNO_Outline-InnerMaskBias}
  n12 = Divide(A=ObjectRadius(), B=ScalarParameter() {defaultValue=97, parameterName=ObjectScale})
  n13 = Saturate(None=Add(A=Subtract(A=@n9.R, B=@n10.R), B=Multiply(A=@n11, B=@n12)))
  n14 = ShadingPathSwitch(Default=Multiply(A=QualitySwitch(Default=@n7, Low=@n8, High=@n8, Medium=@n8, Epic=@n7), B=@n13), Forward=@n13, Mobile=@n13)
  n15 = PixelDepth()
  n16 = ScalarParameter() {defaultValue=50, group=Cartoon, parameterName=CRTNO_InnerLine_FadeStart}
  n17 = Clamp(None=Divide(A=Subtract(A=@n15.R, B=@n16), B=ScalarParameter() {defaultValue=500, group=Cartoon, parameterName=CRTNO_InnerLine_FadeDistance}))
  n18 = ShadingPathSwitch(
   Default=LinearInterpolate(
    A=ScalarParameter() {defaultValue=16, parameterName=SampleBias_Min}
    B=ScalarParameter() {defaultValue=16, parameterName=SampleBias}
    Alpha=@n17
   )
   Mobile=LinearInterpolate(
    A=ScalarParameter() {defaultValue=1, parameterName=MobileSampleBias_Min}
    B=ScalarParameter() {defaultValue=16, parameterName=Mobile_SampleBias}
    Alpha=@n17
   )
  )
  n19 = ComponentMask(None=ScreenPosition().ViewportUV) {g=true, r=true}
  n20 = Subtract(A=Divide(A=ViewSize(), B=Constant() {r=320}), B=Constant() {r=2}) {constB=0}
  n21 = Subtract(A=PixelDepth().R, B=ScalarParameter() {defaultValue=50, group=Cartoon, parameterName=CRTNO_Outline-InnerScaleStart})
  n22 = ScalarParameter() {defaultValue=500, group=Cartoon, parameterName=CRTNO_Outline-InnerScaleDistance}
  n23 = LinearInterpolate(
   A=ScalarParameter() {defaultValue=2, group=Cartoon, parameterName=CRTNO_Outline-InnerSize1}
   B=ScalarParameter() {defaultValue=0.75, group=Cartoon, parameterName=CRTNO_Outline-InnerSize2}
   Alpha=Clamp(None=Divide(A=@n21, B=@n22))
  )
  n24 = Multiply(A=@n23, B=ScalarParameter() {defaultValue=1.5, group=Cartoon, parameterName=CRTNO_Outline-InnerSizeMobileMultiplier})
  n25 = Multiply(
   A=Clamp(None=Ceil(None=@n20)) {maxDefault=2, minDefault=1}
   B=Multiply(A=SceneTexelSize(), B=ShadingPathSwitch(Default=@n23, Forward=@n24, Mobile=@n24))
  ) {constA=2}
  n26 = Constant()
  n27 = AppendVector(A=ComponentMask(None=@n25) {desc=x, r=true}, B=@n26)
  n28 = Add(A=@n19, B=@n27) {desc=x+}
  n29 = SceneDepth(UVs=@n28)
  n30 = AppendVector(A=@n26, B=ComponentMask(None=@n25) {desc=y, g=true}) {desc=y+}
  n31 = Add(A=@n19, B=@n30) {desc=y+}
  n32 = SceneDepth(UVs=@n31)
  n33 = Constant() {r=-1}
  n34 = Add(A=@n19, B=Multiply(A=@n27, B=@n33)) {desc=x-}
  n35 = SceneDepth(UVs=@n34)
  n36 = Add(A=@n19, B=Multiply(A=@n33, B=@n30)) {desc=y-}
  n37 = SceneDepth(UVs=@n36)
  n38 = LinearInterpolate(
   A=ScalarParameter() {defaultValue=-1, parameterName=Divisor_Min}
   B=ScalarParameter() {defaultValue=-16, parameterName=Divisor_max}
   Alpha=@n17
  )
  n39 = LinearInterpolate(
   A=ScalarParameter() {defaultValue=-1, parameterName=MobileDivisor_Min}
   B=ScalarParameter() {defaultValue=-16, parameterName=MobileDivisor_max}
   Alpha=@n17
  )
  n40 = ComponentMask(None=MaterialFunctionCall!(
    Base=@n10.R
    BaseSampleBias=@n18
    Scale=ScalarParameter() {defaultValue=4, parameterName=Scale}
    Sample1=@n29.R
    Sample2=@n32.R
    Sample3=@n35.R
    Sample4=@n37.R
    Divisor=ShadingPathSwitch(Default=@n38, Mobile=@n39)
    UseCenteredLines
    UseRGB
   ) {materialFunction=/Game/Packages/jl_tests/combine4Samples}.Result) {r=true}
  n41 = ScalarParameter() {defaultValue=16, parameterName=SampleBias_Min_RGH}
  n42 = ScalarParameter() {defaultValue=16, parameterName=SampleBias_RGH}
  n43 = ComponentMask(None=MaterialFunctionCall(
    Base=SceneTexture() {sceneTextureId=PPI_Roughness}.Color
    BaseSampleBias=LinearInterpolate(A=@n41, B=@n42, Alpha=@n17)
    Scale=Constant() {r=4}
    Sample1=SceneTexture(UVs=@n28) {sceneTextureId=PPI_Roughness}.Color
    Sample2=SceneTexture(UVs=@n31) {sceneTextureId=PPI_Roughness}.Color
    Sample3=SceneTexture(UVs=@n34) {sceneTextureId=PPI_Roughness}.Color
    Sample4=SceneTexture(UVs=@n36) {sceneTextureId=PPI_Roughness}.Color
    Divisor=ScalarParameter() {defaultValue=-0.04, group=RoughnessGBufferLines, parameterName=InnerLinesDivisor}
   ) {materialFunction=&combine4Samples}.Result) {r=true}
  n44 = StaticSwitchParameter(True=ShadingPathSwitch(Default=Max(A=@n40, B=@n43), Forward=@n40, Mobile=@n40), False=@n40) {group=RoughnessGBufferLines, parameterName=UseRoughnessGBufferOutlines}
  n45 = Multiply(A=Saturate(None=Multiply(A=@n14, B=@n44)), B=ScalarParameter() {defaultValue=1, group=Cartoon, parameterName=CRTNO_Outline-InnerAmount})
  n46 = Constant()
  n47 = StaticSwitchParameter(True=@n45, False=ShadingPathSwitch(Default=@n45, Forward=@n46, Mobile=@n46)) {group=Cartoon, parameterName=UseDepthBufferOutlinesOnMobile}
  n48 = Multiply(A=VertexNormalWS(), B=ScalarParameter() {defaultValue=0.5, parameterName=ScaleMesh})
  n49 = StaticSwitchParameter(True=Add(A=NamedRerouteUsage(), B=@n48), False=@n48) {parameterName=Use WPO}
  MaterialFunctionCall(In=SetMaterialAttributes!(
    MaterialAttributes
    Emissive Color=NamedRerouteDeclaration(None=@n4) {name=emissive, nodeColor=<rgba 1 0.75 0 1>}
    Opacity=NamedRerouteDeclaration(None=@n47) {name=opacity, nodeColor=<rgba 0.4062 1 0 1>}
    World Position Offset=NamedRerouteDeclaration(None=@n49) {name=wpo_1, nodeColor=<rgba 0.1719 0 1 1>}
   ) {attributeSetTypes [#1, #2, #3]}) {materialFunction=&MF_TechArtOverrides}
  n50 = ScalarParameter() {defaultValue=1}
  n51 = ScalarParameter() {defaultValue=-1, parameterName=Param_1}
  n52 = Saturate(None=ComponentMask(None=MaterialFunctionCall(Base=@n10.R, BaseSampleBias=@n18, Scale=@n50, Sample1=@n29.R, Sample2=@n32.R, Sample3=@n35.R, Sample4=@n37.R, Divisor=@n51) {materialFunction=&combine4Samples}.Result) {r=true})
  n53 = ScalarParameter() {defaultValue=1000, parameterName=BaseMult}
  n54 = Multiply(A=ComponentMask(None=SceneTexture() {sceneTextureId=PPI_DiffuseColor}.Color) {b=true}, B=@n53) {constB=1000}
  n55 = ScalarParameter() {parameterName=Param0_2}
  n56 = Constant() {r=4}
  n57 = Multiply(A=ComponentMask(None=SceneTexture(UVs=@n28) {sceneTextureId=PPI_DiffuseColor}.Color) {b=true}, B=@n53) {constB=1000}
  n58 = Multiply(A=ComponentMask(None=SceneTexture(UVs=@n31) {sceneTextureId=PPI_DiffuseColor}.Color) {b=true}, B=@n53) {constB=1000}
  n59 = Multiply(A=ComponentMask(None=SceneTexture(UVs=@n34) {sceneTextureId=PPI_DiffuseColor}.Color) {b=true}, B=@n53) {constB=1000}
  n60 = Multiply(A=ComponentMask(None=SceneTexture(UVs=@n36) {sceneTextureId=PPI_DiffuseColor}.Color) {b=true}, B=@n53) {constB=1000}
  n61 = Saturate(None=ComponentMask(None=MaterialFunctionCall(Base=@n54, BaseSampleBias=@n55, Scale=@n56, Sample1=@n57, Sample2=@n58, Sample3=@n59, Sample4=@n60, Divisor=@n51) {materialFunction=&combine4Samples}.Result) {r=true})
  NamedRerouteDeclaration(None=Multiply(A=OneMinus(None=@n52), B=OneMinus(None=@n61))) {name=Outline Color Mask, nodeColor=<rgba 1 0 0.9982 1>}
  n62 = VectorParameter() {parameterName=BobHeightMax, defaultValue=<rgba 0 0 10 0>}
  n63 = VectorParameter() {parameterName=BobHeightMin, defaultValue=<rgba 0 0 -2 0>}
  n64 = Multiply(A=Time(), B=ScalarParameter() {defaultValue=0.4, parameterName=WPOTimeScale})
  n65 = Sine(None=@n64)
  n66 = Constant3Vector() {constant=<rgba 0 0 1 0>}
  n67 = Multiply(
   A=LinearInterpolate(
    A=ScalarParameter() {defaultValue=-0.01, parameterName=Rotation Angle Min}
    B=ScalarParameter() {defaultValue=0.01, parameterName=Rotation Angle Max}
    Alpha=Cosine(None=@n64)
   )
   B=OneMinus(None=VectorParameter() {parameterName=WorldPostionOfBlackHole, defaultValue=<rgba -30000 25536 1952 0>}.A)
  )
  n68 = ObjectPositionWS()
  n69 = WorldPosition()
  NamedRerouteDeclaration(None=Add(
    A=LinearInterpolate(A=@n62.RGB, B=@n63.RGB, Alpha=@n65)
    B=RotateAboutAxis(NormalizedRotationAxis=@n66, RotationAngle=@n67, PivotPoint=@n68, Position=@n69.XYZ)
   )) {name=WPO, nodeColor=<rgba 0.3359 0 1 1>}
  n70 = Multiply(
   A=Subtract(A=WorldPosition().XYZ, B=ObjectPositionWS())
   B=ScalarParameter() {defaultValue=0.001, group=Emissive Sweep, parameterName=WorldTileScale}
  )
  n71 = Normalize(VectorInput=Transform(None=VectorParameter() {group=Emissive Sweep, parameterName=SweepDirectionCamSpace, defaultValue=<rgba -0.5 0.45 -1 0>}.RGB) {transformSourceType=TRANSFORMSOURCE_Camera})
  n72 = ComponentMask(None=Multiply(A=DotProduct(A=@n70, B=@n71), B=@n71)) {b=true, r=true}
  n73 = Constant() {r=0.28}
  n74 = Frac(None=ComponentMask(None=Panner(Coordinate=@n72, Speed=@n73) {speedX=0.2, speedY=0.2}) {g=true})
  n75 = Add(A=@n74) {constB=0.1}
  n76 = OneMinus(None=Multiply(A=Subtract(A=@n74) {constB=0.9}) {constB=10})
  n77 = MaterialFunctionCall() {materialFunction=&MF_Fresnel_FX}
  n78 = Multiply(A=Multiply(A=Floor(None=@n75), B=Power(Base=@n76) {constExponent=3}), B=Saturate(None=Add(A=@n77.Fresnel Inverted) {constB=0.1}))
  n79 = VectorParameter() {group=Emissive Sweep, parameterName=SweepColour, defaultValue=<rgba 1.8958 2 1.9078 0>}
  n80 = LinearInterpolate(
   A=@n79.RGB
   B=Multiply(A=@n79.RGB, B=ScalarParameter() {defaultValue=15, group=Emissive Sweep, parameterName=FBoostScale})
   Alpha=VertexColor().G
  )
  NamedRerouteDeclaration(None=StaticSwitchParameter(True=Multiply(A=@n78, B=@n80), False=Constant()) {group=Emissive Sweep, parameterName=UseSweep}) {name=EmissiveSweep, nodeColor=<rgba 1 0.5859 0 1>}
  SceneTexture()
 ]
}
